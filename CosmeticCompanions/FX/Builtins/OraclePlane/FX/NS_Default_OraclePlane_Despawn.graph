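# Niagara despawn effect for the default OraclePlane cosmetic companion, built around a burst of colored, fluttering sprites.
# BlowingParticles launches a few dozen sprites that tumble under gravity and aerodynamic drag, using mesh orientation for alignment and SubUV animation.
# A waveform flickers their brightness, while randomly selected pulse curves write a custom Vector2D alongside two sprite renderers.
# AirglowLWE and SmokeLWE expose no renderers here, and SparklesLWE is disabled.
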NiagaraSystem /CosmeticCompanions/FX/Builtins/OraclePlane/FX/NS_Default_OraclePlane_Despawn
dependencies {
 usedDataInterfaces [
  &NiagaraDataInterfaceArrayColor
  &NiagaraDataInterfaceCurve
  &NiagaraDataInterfaceSpriteRendererInfo
 ]
 usedDynamicInputs []
 usedModules [&SystemState]
}
emitters {
 AirglowLWE {rendererClasses [], renderers []}
 BlowingParticles {
  emitterUpdateScript [
   &EmitterState(Life Cycle Mode=::NewEnumerator1, Loop Behavior=::NewEnumerator1, UseLoopDelay=false)
   &SpawnBurst_Instantaneous(
    Spawn Count=&RandomRangeInt(Minimum=25, Maximum=35)
    Use Spawn Probability=false
    UseLoopCountLimit=false
   )
  ]
  particleSpawnScript [
   &InitializeParticle(
    Color=&SelectLinearColorFromArray(
     Array Sampling Mode=::NewEnumerator0
     Direct Array Index=?
     Color Selection Array=DataInterface() {propertyValues {ColorData [<rgba 0.8958 0.098 0.752 1>, <rgba 0.0156 0.4736 1 1>, <rgba 0 2 1.8173 1>]}}
     Random Seed=0
    )
    Lifetime Min=3
    Lifetime Mode=::NewEnumerator1
    Mass Max=1.5
    Mass Mode=::NewEnumerator3
    Sprite Rotation Mode=::NewEnumerator2
    Sprite Size Mode=::NewEnumerator1
    Uniform Sprite Size Max=2
    UsePositionOffset=false
    Write Material Random=true
   )
   &InitialMeshOrientation(Use Rotation Vector=false)
   &AddVelocity(Module.Use Falloff Distance=false, Override Seed=false, Velocity Mode=::NewEnumerator1, Velocity Speed=297.6503)
   &ShapeLocation(MaskSpawnGroup=false, Override Seed=false, Sphere Radius=10, UseOwnerScale=true)
  ]
  particleUpdateScript [
   &ParticleState()
   &GravityForce(Gravity=<0 0 -1108.3575>)
   &AerodynamicDrag(
    Aerodynamic Drag:NiagaraFloat=&RandomRangeFloat(Minimum=1.5, Maximum=2.5)
    Aerodynamic Rotational Drag=0.15
    Lift Contribution=0.75
    Pivot Offset Maximum:Vector3f=<5 0.5 1>
    Pivot Offset Minimum:Vector3f=<-5 -0.5 -1>
    Rotation Strength:NiagaraFloat=&RandomRangeFloat(Minimum=1.8, Maximum=3.2)
    Use Linear Drag=true
    Use Rotational Drag=true
   )
   &Drag(
    Drag=&FloatFromCurve(
     FloatCurve=DataInterface() {propertyValues {Curve {
        defaultValue=0
        keys [
         {arriveTangent=4.1785, interpMode=RCIM_Cubic, leaveTangent=4.1785, tangentMode=RCTM_User, time=0, value=0}
         {arriveTangent=-0.5407, interpMode=RCIM_Cubic, leaveTangent=-0.5407, tangentMode=RCTM_User, time=0.1001, value=1}
         {arriveTangent=-0.2573, interpMode=RCIM_Cubic, leaveTangent=-0.2573, tangentMode=RCTM_User, time=0.1695, value=0.068}
         {arriveTangent=0.0224, interpMode=RCIM_Cubic, leaveTangent=0.0224, tangentMode=RCTM_User, time=1, value=0}
        ]
       }}}
     Scale Curve=1.28
    )
    Use Linear Drag=true
    Use Rotational Drag=true
   )
   &SolveForcesAndVelocity(Clamp Velocity=false, Limit Acceleration=false)
   &ScaleSpriteSize(Uniform Curve Sprite Scale=DataInterface() {propertyValues {Curve {keys [
        {arriveTangent=14.5835, interpMode=RCIM_Cubic, leaveTangent=14.5835, tangentMode=RCTM_User, time=0, value=0}
        {arriveTangent=1.1111, interpMode=RCIM_Cubic, leaveTangent=1.1111, time=0.1, value=1}
        {arriveTangent=-1.1111, interpMode=RCIM_Cubic, leaveTangent=-1.1111, time=0.9, value=1}
        {arriveTangent=-8.7113, interpMode=RCIM_Cubic, leaveTangent=-8.7113, tangentMode=RCTM_User, time=1, value=0}
       ]}}})
   &AlignSpriteToMeshOrientation(SetAlignment=true, SetFacing=true)
   &SubUVAnimation(
    Random Start Frame=true
    "Sprite Renderer "=DataInterface() {propertyValues {SpriteRenderer=&self:BlowingParticles.NiagaraSpriteRendererProperties_0}}
    SubUV Lookup Index Scale:NiagaraFloat=&RandomRangeFloat(Minimum=0.5, Maximum=3)
    UseEndFrame=false
    UseStartFrame=false
   )
   /SetVariables(Particles.Vector 2D:Vector2f=&MergeHelper_Vector2Switch(
     0=&Multiply_Vector2D_ByFloat(
      Float=&FloatFromCurve(FloatCurve=DataInterface() {propertyValues {Curve {keys [<at 0.095 0.0337>,<at 0.1374 0.9214>,<at 0.1442 0.0735>,<at 1 0>]}}})
      Vector2D=<30 30>
     )
     1=&Multiply_Vector2D_ByFloat(
      Float=&FloatFromCurve(FloatCurve=DataInterface() {propertyValues {Curve {keys [<at 0.3586 0.0118>,<at 0.4089 0.9324>,<at 0.4402 0.0735>,<at 0.5421 0.0643>,<at 0.5813 0.5638>,<at 0.6119 0.0487>,<at 1 0>]}}})
      Vector2D=<30 30>
     )
     2=&Multiply_Vector2D_ByFloat(
      Float=&FloatFromCurve(FloatCurve=DataInterface() {propertyValues {Curve {keys [<at 0 0>,<at 0.0704 0.69>,<at 0.085 0>,<at 0.775 0.0118>,<at 0.8253 0.9324>,<at 0.8566 0.0735>,<at 1 0>]}}})
      Vector2D=<30 30>
     )
     3=&Multiply_Vector2D_ByFloat(
      Float=&FloatFromCurve(FloatCurve=DataInterface() {propertyValues {Curve {keys [<at 0.0626 0.0118>,<at 0.1129 0.9324>,<at 0.1442 0.0735>,<at 0.5484 0.0475>,<at 0.6041 1.3152>,<at 0.6439 -0.0076>,<at 1 0>]}}})
      Vector2D=<15 15>
     )
     4=?
     Index=&RandomRangeInt(Minimum=0, Maximum=3)
     Num=::NewEnumerator2
    )) {isSetParameters=true}
   &ScaleColor(
    Scale RGB=&VectorFromFloat(Value=&Waveform(
      "Waveform [1]"=::NewEnumerator9
      "[1] Frequency"=&RandomRangeFloat(Minimum=15.3789, Maximum=28)
      "[1] Amplitude Min/Max"=<1 8.4307>
     ))
    ScaleA=true
    ScaleRGB=true
   )
  ]
  rendererClasses [&NiagaraSpriteRendererProperties]
  renderers [
   {rendererClass=&NiagaraSpriteRendererProperties}
   {rendererClass=&NiagaraSpriteRendererProperties, rendererIndex=1}
  ]
 }
 SmokeLWE {rendererClasses [], renderers []}
 SparklesLWE {bEnabled=false, rendererClasses [], renderers []}
}
summary {
 systemName=NS_Default_OraclePlane_Despawn
 emitters [
  {bEnabled=false, emitterName=SparklesLWE, rendererClasses []}
  {emitterName=AirglowLWE, rendererClasses []}
  {emitterName=SmokeLWE, rendererClasses []}
  {emitterName=BlowingParticles, rendererClasses [&NiagaraSpriteRendererProperties]}
 ]
 userVariables [User.NormalizedNotifyProgress {type=NiagaraFloat, defaultValue=0, description=""}]
}
