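# Niagara system for the disco ball effect of the OraclePlane emote, pairing six lightweight emitters (airglow, smoke, sparkles) with a detailed BlowingParticles emitter.
# BlowingParticles loops short bursts at one or two preset offsets indexed by loop count, also cycling channel-mask colors for star picking.
# Spawn groups separate stationary central sprites from faster outward particles slowed by curve-driven drag.
# Sprite and light renderers use pink glow, flickering bokeh size and fading light attributes.
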NiagaraSystem /CosmeticCompanions/FX/Builtins/OraclePlane/FX/NS_Emote_OraclePlane_DiscoBall
dependencies {
 usedDataInterfaces [
  &NiagaraDataInterfaceArrayColor
  &NiagaraDataInterfaceArrayFloat3
  &NiagaraDataInterfaceCurve
  &NiagaraDataInterfaceVector2DCurve
  &NiagaraDataInterfaceColorCurve
 ]
 usedDynamicInputs [&Add_Float]
 usedModules [
  &self:SystemSpawnScript.SystemScriptSource.SystemScriptGraph.NiagaraNodeAssignment_1.SetVariables_#1
  &self:SystemSpawnScript.SystemScriptSource.SystemScriptGraph.NiagaraNodeAssignment_2.SetVariables_#2
  &self:SystemSpawnScript.SystemScriptSource.SystemScriptGraph.NiagaraNodeAssignment_3.SetVariables_#3
  &self:SystemSpawnScript.SystemScriptSource.SystemScriptGraph.NiagaraNodeAssignment_4.SetVariables_#4
  &SystemState
 ]
}
emitters {
 AirglowLWE {rendererClasses [], renderers []}
 BlowingParticles {
  emitterUpdateScript [
   &EmitterState(Life Cycle Mode=::NewEnumerator1, Loop Behavior=::NewEnumerator2, Loop Count=15, Loop Duration=0.55, UseLoopDelay=false)
   /SetVariables(Emitter.RandomPicker:NiagaraInt32=&RandomRangeInt(Minimum=1, Maximum=2)) {isSetParameters=true}
   /SetVariables(
    Emitter.StarGlowPicker:LinearColor=&SelectLinearColorFromArray(
     Array Sampling Mode=::NewEnumerator1
     Direct Array Index=$Emitter.LoopCount
     Direct Array Mode=::NewEnumerator1
     Recalculate Random Each Loop=true
     Color Selection Array=DataInterface() {propertyValues {ColorData [<rgba 0 0 1 0>, <rgba 0 1 0 0>, <rgba 1 0 0 0>]}}
     Random Seed=?
    )
    Emitter.StarPicker:LinearColor=&SelectLinearColorFromArray(
     Array Sampling Mode=::NewEnumerator1
     Direct Array Index=$Emitter.LoopCount
     Direct Array Mode=::NewEnumerator1
     Recalculate Random Each Loop=true
     Color Selection Array=DataInterface() {propertyValues {ColorData [<rgba 0 0 1 0>, <rgba 0 1 0 0>, <rgba 1 0 0 0>]}}
     Random Seed=?
    )
   ) {isSetParameters=true}
   &SpawnBurst_Instantaneous(Spawn Count=$Emitter.RandomPicker, Use Spawn Probability=false, UseLoopCountLimit=false)
   &SpawnBurst_Instantaneous(
    Spawn Count=&MergeHelper_IntSwitch(
     0=0
     1=&RandomRangeInt(Minimum=5, Maximum=7)
     2=&RandomRangeInt(Minimum=10, Maximum=14)
     3=?
     4=?
     5=?
     Index=$Emitter.RandomPicker
     Num=::NewEnumerator1
    )
    Spawn Group=1
    Use Spawn Probability=false
    UseLoopCountLimit=false
   ) {module=SpawnBurst_Instantaneous003}
   &SpawnBurst_Instantaneous(
    Spawn Count=&RandomRangeInt(Minimum=2, Maximum=5)
    Spawn Group=2
    Use Spawn Probability=false
    UseLoopCountLimit=false
   ) {module=SpawnBurst_Instantaneous001}
  ]
  particleSpawnScript [
   /SetVariables(Particles.VisibilityTag=&SpawnGroup()) {isSetParameters=true}
   &PartitionParticles(Number Of Partitions=$Emitter.RandomPicker)
   /SetVariables() {isSetParameters=true}
   &InitializeParticle(
    Color=<rgba 1 1 1 0.887>
    Lifetime=&MergeHelper_FloatSwitch(
     0=0.5
     1=&RandomRangeFloat(Minimum=0.1, Maximum=0.4)
     2=&RandomRangeFloat(Minimum=0.3, Maximum=0.6)
     3=?
     Index=&SpawnGroup()
     Num=::NewEnumerator1
    )
    Mass Max=1.5
    Mass Mode=::NewEnumerator3
    Sprite Size Mode=::NewEnumerator3
    Uniform Sprite Size=37.3633
    UsePositionOffset=false
    Write Material Random=true
   )
   &ShapeLocation(
    MaskSpawnGroup=false
    Offset=&MergeHelper_VectorSwitch(
     0=&SelectVectorFromArray(
      Direct Array Index=$Emitter.LoopCount
      Direct Array Mode=::NewEnumerator1
      Vector Selection Array=DataInterface() {propertyValues {FloatData [<-5.5216 62.4477 100.9861>, <61.0396 3.2644 54.1851>, <-15.1425 46.0963 71.1649>, <62.4558 -4.5395 102.5931>, <-73.1112 5.9684 79.9621>]}}
     )
     1=&SelectVectorFromArray(
      Direct Array Index=$Emitter.LoopCount
      Direct Array Mode=::NewEnumerator1
      Vector Selection Array=DataInterface() {propertyValues {FloatData [<23.4319 -33.3874 66.8765>, <-12.27 41.2272 73.5568>, <-19.5124 -66.4395 94.7768>, <-51.5635 59.4972 73.5568>, <13.1183 50.3289 50.5711>]}}
     )
     2=?
     3=?
     Index=$Particles.PartitionParticles.Partition
     Num=::NewEnumerator0
    )
    Offset Mode=::NewEnumerator0
    Override Seed=false
    Sphere Radius=0
    UseOwnerScale=true
   )
   &AddVelocity(
    Module.Use Falloff Distance=false
    Override Seed=false
    Velocity Mode=::NewEnumerator1
    Velocity Origin=$Particles.ShapeLocation.ShapePosition
    Velocity Speed=&MergeHelper_FloatSwitch(0=0, 1=257.982, 2=297.9542, 3=?, Index=&SpawnGroup(), Num=::NewEnumerator1)
   )
   &AddVelocity(Module.Use Falloff Distance=false, Override Seed=false, Velocity Mode=::NewEnumerator1, Velocity Speed=33.2933) {module=AddVelocity001}
  ]
  particleUpdateScript [
   &ParticleState()
   &ScaleSpriteSize(Uniform Curve Sprite Scale=DataInterface() {propertyValues {Curve {keys [
        {arriveTangent=2.3879, interpMode=RCIM_Cubic, leaveTangent=2.3879, tangentMode=RCTM_User, time=0, value=0}
        {arriveTangent=-0.1058, interpMode=RCIM_Cubic, leaveTangent=-0.1058, tangentMode=RCTM_User, time=0.9615, value=0.9815}
       ]}}})
   /SetVariables(Particles.GlowColor=&Multiply_LinearColorByFloat(
     Linear Color=<rgba 1 0 0.8022 0.6261>
     Float=&FloatFromCurve(FloatCurve=DataInterface() {propertyValues {Curve {
         defaultValue=1
         keys [
          {arriveTangent=-2.5249, interpMode=RCIM_Cubic, leaveTangent=-2.5249, tangentMode=RCTM_User, time=0, value=2}
          {arriveTangent=-0.0552, interpMode=RCIM_Cubic, leaveTangent=-0.0552, tangentMode=RCTM_User, time=0.0504, value=1}
          {arriveTangent=-1.0531, interpMode=RCIM_Cubic, leaveTangent=-1.0531, time=0.8517, value=1}
          {arriveTangent=-0.0405, interpMode=RCIM_Cubic, leaveTangent=-0.0405, tangentMode=RCTM_User, time=1, value=0}
         ]
        }}})
    )) {isSetParameters=true}
   /SetVariables(Particles.MotesSize:Vector2f=&Vector2DFromCurve(
     Vector2Curve=DataInterface() {propertyValues {XCurve {keys [<at 0.6 1>,<at 1 0>]}, YCurve {keys [<at 0.6 1>,<at 1 0>]}}}
     ScaleCurve=&Multiply_Vector2D_ByFloat(Float=&RandomRangeFloat(Minimum=2, Maximum=4), Vector2D=<0.5 3>)
    )) {isSetParameters=true}
   /SetVariables(
    Particles.BokehColor:LinearColor=&ColorFromCurve(ColorCurve=DataInterface() {propertyValues {
       AlphaCurve {keys [<at 0.3427 0.2 RCIM_Cubic>,<at 1 0 RCIM_Cubic>]}
       BlueCurve {keys [<at 0 0.7886 RCIM_Cubic>]}
       GreenCurve {keys [<at 0 0 RCIM_Cubic>]}
       RedCurve {keys [<at 0 1 RCIM_Cubic>]}
      }})
    Particles.BokehSize:Vector2f=&Multiply_Vector2D_ByFloat(
     Float=&FloatFromCurve(
      FloatCurve=DataInterface() {propertyValues {Curve {keys [<at 0 0>,<at 0.2214 1.0265>,<at 0.8106 1>,<at 1 0>]}}}
      Scale Curve=&Multiply_Float(
       A=&RandomRangeFloat(Minimum=5, Maximum=12.8834)
       B=&Waveform("Waveform [1]"=::NewEnumerator9, "[1] Frequency"=10, "[1] Amplitude Min/Max"=<0.5 1>, "[1] Scale"=1.5008)
      )
     )
     Vector2D=<1 1>
    )
   ) {isSetParameters=true}
   &Drag(
    Drag=&MergeHelper_FloatSwitch(
     0=0
     1=&FloatFromCurve(
      FloatCurve=DataInterface() {propertyValues {Curve {
         defaultValue=1
         keys [
          {arriveTangent=4.1785, interpMode=RCIM_Cubic, leaveTangent=4.1785, tangentMode=RCTM_User, time=0, value=0}
          {arriveTangent=0.1754, interpMode=RCIM_Cubic, leaveTangent=0.1754, tangentMode=RCTM_User, time=0.6699, value=0.9374}
          {arriveTangent=0.0224, interpMode=RCIM_Cubic, leaveTangent=0.0224, tangentMode=RCTM_User, time=1, value=0.9412}
         ]
        }}}
      Scale Curve=&RandomRangeFloat(Minimum=12, Maximum=14)
     )
     2=&FloatFromCurve(
      FloatCurve=DataInterface() {propertyValues {Curve {
         defaultValue=1
         keys [
          {arriveTangent=4.1785, interpMode=RCIM_Cubic, leaveTangent=4.1785, tangentMode=RCTM_User, time=0, value=0}
          {arriveTangent=0.1754, interpMode=RCIM_Cubic, leaveTangent=0.1754, tangentMode=RCTM_User, time=0.2076, value=0.9021}
          {arriveTangent=0.0224, interpMode=RCIM_Cubic, leaveTangent=0.0224, tangentMode=RCTM_User, time=1, value=1}
         ]
        }}}
      Scale Curve=&RandomRangeFloat(Minimum=18, Maximum=24.8712)
     )
     3=0
     4=0
     Index=&SpawnGroup()
     Num=::NewEnumerator3
    )
    Use Linear Drag=true
    Use Rotational Drag=true
   )
   &SolveForcesAndVelocity(Clamp Velocity=false, Limit Acceleration=false)
   &ScaleColor(
    Linear Color Curve=DataInterface() {propertyValues {AlphaCurve {keys [<at 0 1>,<at 1 0>]}, BlueCurve {keys [<at 0 1>]}, GreenCurve {keys [<at 0 1>]}, RedCurve {keys [<at 0 1>]}}}
    Scale Mode=2
   )
   &FN_Light_Attributes(
    Color=<rgba 10 1.163 9.6353 1>
    Create Color Attribute=true
    Create Position Attribute=true
    Create Radius Attribute=true
    Position:NiagaraPosition=$Particles.ShapeLocation.ShapePosition
    Radius=80
    Scale Color=&FloatFromCurve(FloatCurve=DataInterface() {propertyValues {Curve {keys [<at 0 1>,<at 1 0>]}}})
   )
  ]
  rendererClasses [&NiagaraSpriteRendererProperties, &NiagaraLightRendererProperties]
  renderers [
   {rendererClass=&NiagaraSpriteRendererProperties}
   {rendererClass=&NiagaraSpriteRendererProperties, rendererIndex=1}
   {rendererClass=&NiagaraSpriteRendererProperties, rendererIndex=2}
   {rendererClass=&NiagaraSpriteRendererProperties, rendererIndex=3}
   {rendererClass=&NiagaraLightRendererProperties, rendererIndex=4}
  ]
 }
 SmokeLWE {rendererClasses [], renderers []}
 SparklesLWE {rendererClasses [], renderers []}
 SparklesLWE001 {rendererClasses [], renderers []}
 SparklesLWE002 {rendererClasses [], renderers []}
 SparklesLWE003 {rendererClasses [], renderers []}
}
summary {
 systemName=NS_Emote_OraclePlane_DiscoBall
 emitters [
  {emitterName=SparklesLWE, rendererClasses []}
  {emitterName=AirglowLWE, rendererClasses []}
  {emitterName=SmokeLWE, rendererClasses []}
  {emitterName=SparklesLWE001, rendererClasses []}
  {emitterName=SparklesLWE002, rendererClasses []}
  {emitterName=SparklesLWE003, rendererClasses []}
  {
   emitterName=BlowingParticles
   rendererClasses [&NiagaraSpriteRendererProperties, &NiagaraLightRendererProperties]
  }
 ]
 userVariables []
}
